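# Material function that samples Fortnite real-time wind data in world space to provide water foam, a height map and a ripple normal.
# Foam is the inverted animated-texture alpha at the pixel's world position.
# Blue-channel heights at ±X and ±Y offsets give central-difference slopes for a normal, blended from flat up by an intensity input.
# The Height Map output reuses the +X offset sample, not a centered one.
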MaterialFunction /Figment_S10/WaterAndWind/MaterialFunctions/Fortnite_SampleRealTimeWaterData
graph {
 properties {bExposeToLibrary=true, libraryCategoriesText [Fortnite Wind]}
 nodes [
  n1 = WorldPosition()
  FunctionOutput(None=OneMinus(None=MaterialFunctionCall!("World Position (See Note)"=@n1.XYZ, Max Rotation) {materialFunction=&/Fortnite_SampleRealTimeWindData}.Animated Texture Alpha Output)) {outputName=Foam}
  n2 = Constant3Vector() {constant=<rgba 0 0 1 0>}
  n3 = Constant3Vector() {constant=<rgba -1 -1 1 0>}
  n4 = FunctionInput(Preview=ScalarParameter() {defaultValue=6, group=Water Ripple, parameterName="World space unit offset to calculate water normal "}) {bUsePreviewValueAsDefault=true, inputName="World space offset to calculate normal ", inputType=FunctionInput_Scalar}
  n5 = Multiply(A=Constant3Vector() {constant=<rgba 1 0 0 0>}, B=@n4)
  n6 = ComponentMask(None=MaterialFunctionCall("World Position (See Note)"=Add(A=@n1.XYZ, B=@n5)) {materialFunction=&Fortnite_SampleRealTimeWindData}.Animated Texture Output) {b=true}
  n7 = Subtract(A=@n1.XYZ, B=@n5)
  n8 = Multiply(A=@n4, B=Constant3Vector() {constant=<rgba 0 1 0 0>})
  n9 = Add(A=@n1.XYZ, B=@n8)
  n10 = Subtract(A=@n1.XYZ, B=@n8)
  n11 = MaterialFunctionCall(
   X=@n6
   Y=ComponentMask(None=MaterialFunctionCall("World Position (See Note)"=@n7) {materialFunction=&Fortnite_SampleRealTimeWindData}.Animated Texture Output) {b=true}
   Z=ComponentMask(None=MaterialFunctionCall("World Position (See Note)"=@n9) {materialFunction=&Fortnite_SampleRealTimeWindData}.Animated Texture Output) {b=true}
   A=ComponentMask(None=MaterialFunctionCall("World Position (See Note)"=@n10) {materialFunction=&Fortnite_SampleRealTimeWindData}.Animated Texture Output) {b=true}
  ) {materialFunction=&MakeFloat4}
  n12 = Subtract(A=ComponentMask(None=@n11.Result) {b=true, r=true}, B=ComponentMask(None=@n11.Result) {a=true, g=true}) {desc=HeightDeltas}
  n13 = Multiply(A=@n4) {constB=2}
  n14 = MaterialFunctionCall(X=ComponentMask(None=@n12) {r=true}, Y=@n13, Z=@n13) {materialFunction=&MakeFloat3}
  n15 = ComponentMask(None=@n12) {g=true}
  n16 = Normalize(VectorInput=Multiply(
    A=Multiply(A=@n3, B=@n14.Result)
    B=MaterialFunctionCall(X=@n13, Y=@n15, Z=@n13) {materialFunction=&MakeFloat3}.Result
   ))
  n17 = FunctionInput(Preview=ScalarParameter() {defaultValue=0.4, group=Water Ripple, parameterName=Water Ripple Normal Intensity}) {bUsePreviewValueAsDefault=true, inputName=Normal intensity, inputType=FunctionInput_Scalar}
  FunctionOutput(None=Normalize(VectorInput=LinearInterpolate(A=@n2, B=@n16, Alpha=@n17))) {outputName=Normal from height map}
  FunctionOutput(None=@n6) {outputName=Height Map}
 ]
}
